# Copyright Epic Games, Inc. All Rights Reserved.
#################################################
# Generated Digest of Verse API
# DO NOT modify this manually!
# Generated from build: ++Fortnite+Release-36.20-CL-43847582
#################################################


using {/Verse.org/SceneGraph}
using {/UnrealEngine.com/Temporary/UI}
using {/Verse.org/Simulation}
using {/Verse.org/Assets}
M_BasicMaterial_material<scoped {/playkigai.dig@fortnite.com/Minigame_SnowballWarfare}> := class<final><scoped {/playkigai.dig@fortnite.com/Minigame_SnowballWarfare}>(material):
    @editable
    var Metallic<public>:float = external {}

    @editable
    var Specular<public>:float = external {}

    @editable
    var Roughness<public>:float = external {}

M_BasicMaterial<scoped {/playkigai.dig@fortnite.com/Minigame_SnowballWarfare}>:material = external {}

MI_ReadyPlatform_material<scoped {/playkigai.dig@fortnite.com/Minigame_SnowballWarfare}> := class<final><scoped {/playkigai.dig@fortnite.com/Minigame_SnowballWarfare}>(material):
    @editable
    var Metallic<public>:float = external {}

    @editable
    var Specular<public>:float = external {}

    @editable
    var Roughness<public>:float = external {}

MI_ReadyPlatform<scoped {/playkigai.dig@fortnite.com/Minigame_SnowballWarfare}>:material = external {}

MI_ReadyPlatformButton_material<scoped {/playkigai.dig@fortnite.com/Minigame_SnowballWarfare}> := class<final><scoped {/playkigai.dig@fortnite.com/Minigame_SnowballWarfare}>(material):
    @editable
    var Metallic<public>:float = external {}

    @editable
    var Specular<public>:float = external {}

    @editable
    var Roughness<public>:float = external {}

MI_ReadyPlatformButton<scoped {/playkigai.dig@fortnite.com/Minigame_SnowballWarfare}>:material = external {}

MI_SnowballPile_material<scoped {/playkigai.dig@fortnite.com/Minigame_SnowballWarfare}> := class<final><scoped {/playkigai.dig@fortnite.com/Minigame_SnowballWarfare}>(material):
    @editable
    var Metallic<public>:float = external {}

    @editable
    var Specular<public>:float = external {}

    @editable
    var Roughness<public>:float = external {}

MI_SnowballPile<scoped {/playkigai.dig@fortnite.com/Minigame_SnowballWarfare}>:material = external {}

NewWidgetBlueprint<scoped {/playkigai.dig@fortnite.com/Minigame_SnowballWarfare}> := class<public>(widget):


smallplatform<scoped {/playkigai.dig@fortnite.com/Minigame_SnowballWarfare}>:mesh = external {}

snowball_game_manager_asset<scoped {/playkigai.dig@fortnite.com/Minigame_SnowballWarfare}>:entity_prefab = external {}

snowball_pile_asset<scoped {/playkigai.dig@fortnite.com/Minigame_SnowballWarfare}>:entity_prefab = external {}

snowballpile<scoped {/playkigai.dig@fortnite.com/Minigame_SnowballWarfare}>:mesh = external {}